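# Additive, unlit, two-sided particle material for the Wukong pickaxe's helix energy effect, built for sprite and mesh particles.
# Desaturated, distorted panning noise multiplies a gradient mask and particle color, with dynamic parameters driving mask tiling, offset and looping pan.
# Squared UV falloffs, switchable between linear and spherical, multiply particle alpha.
# An optional SplineThicken world position offset takes its width from a dynamic parameter.
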Material /BRCosmetics/Weapons/FORT_Melee/Pickaxe_WuKong/FX/M_Spline_Helix_Energy
graph {
 properties {
  bUsedWithMeshParticles=true
  bUsedWithNiagaraMeshParticles=true
  bUsedWithNiagaraSprites=true
  bUsedWithParticleSprites=true
  blendMode=BLEND_Additive
  pixelDepthOffsetMode=PDOM_Legacy
  shadingModel=MSM_Unlit
  twoSided=true
  usedShadingModels=Unlit
 }
 nodes [
  n1 = TextureCoordinate() {uTiling=0.5, vTiling=0.5}
  n2 = TextureSampleParameter2D(UVs=@n1) {parameterName=DistortNoise, texture=&T_NoisePack_01}
  n3 = ShadingPathSwitch(Default=Divide(A=ComponentMask(None=@n2.RGB) {g=true, r=true}) {constB=10}, Mobile=Constant()) {desc=Migrated from FeatureLevelSwitch}
  n4 = Panner(Coordinate=@n1) {speedY=-0.2}
  n5 = Desaturation(
   None=TextureSampleParameter2D(UVs=Add(A=@n3, B=@n4)) {parameterName=BasicNoise, texture=&T_NoisePack_01}.RGB
   Fraction=Constant() {r=0.82}
  )
  n6 = MaterialFunctionCall(Float2=TextureCoordinate() {vTiling=0.6}) {materialFunction=&BreakOutFloat2Components}
  n7 = DynamicParameter() {paramNames [Width,Random,Vtile,PanSpeed]}
  n8 = OneMinus(None=Multiply(A=@n6.G, B=@n7.Vtile))
  n9 = Add(A=Multiply(A=Time(), B=@n7.PanSpeed), B=@n7.Random)
  n10 = TextureSampleParameter2D(UVs=Panner(
    Coordinate=AppendVector(A=@n6.R, B=@n8)
    Time=StaticSwitchParameter(True=@n9, False=@n7.Random) {defaultValue=true, parameterName=LoopUVanim}
   ) {speedY=1}) {parameterName=Mask, samplerSource=SSM_Wrap_WorldGroupSettings, texture=/Game/Effects/Fort_Effects/Textures/Masks/T_Vert_Gradient_Softer}
  MaterialFunctionCall(Emissive=Multiply(A=ParticleColor().RGB, B=Multiply(A=@n5, B=@n10.R))) {materialFunction=&MF_FX_EmissiveAdjustments}
  n11 = TextureCoordinate()
  n12 = OneMinus(None=ComponentMask(None=@n11) {g=true})
  n13 = Multiply(A=@n12, B=@n12)
  n14 = SphereMask(A=TextureCoordinate(), B=Constant() {r=0.5}) {attenuationRadius=0.5, hardnessPercent=0}
  n15 = Multiply(A=@n14, B=@n14)
  n16 = ComponentMask(None=@n11) {r=true}
  n17 = Saturate(None=Sine(None=Multiply(A=@n16) {constB=0.5}))
  Multiply(
   A=Multiply(A=StaticSwitchParameter(True=@n13, False=@n15) {defaultValue=true, parameterName=SphereMask}, B=Multiply(A=@n17, B=@n17))
   B=ParticleColor().A
  )
  n18 = LinearInterpolate(B=ScalarParameter() {defaultValue=10, parameterName=Width}, Alpha=DynamicParameter() {paramNames [Width,Random,Vtile,PanSpeed]}.Width)
  StaticSwitchParameter(
   True=MaterialFunctionCall(WidthBase=@n18, WidthTip=@n18) {materialFunction=&SplineThicken}.WorldPositionOffset
   False=Constant()
  ) {defaultValue=true, parameterName=UseSpline}
 ]
}
