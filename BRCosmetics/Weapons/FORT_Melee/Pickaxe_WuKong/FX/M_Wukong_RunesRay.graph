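# Unlit, two-sided additive material for mesh particles, named for the rune ray effect on the Wukong pickaxe.
# Panning smoke and caustics textures, masked by squared vertex red and a depth-fade intersection band, blend dark orange into gold tinted by particle color.
# Layered time sines flicker the HDR emissive range.
# Particle alpha, vertex mask and inverted Fresnel form a fade term; mobile drops caustics, depth band, flicker and Fresnel.
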Material /BRCosmetics/Weapons/FORT_Melee/Pickaxe_WuKong/FX/M_Wukong_RunesRay
graph {
 properties {
  bEnableResponsiveAA=true
  bUsedWithMeshParticles=true
  bUsedWithNiagaraMeshParticles=true
  blendMode=BLEND_Additive
  pixelDepthOffsetMode=PDOM_Legacy
  shadingModel=MSM_Unlit
  twoSided=true
  usedShadingModels=Unlit
 }
 nodes [
  n1 = Panner(Coordinate=TextureCoordinate() {uTiling=0.2, vTiling=0.5}) {speedX=0.5, speedY=0.2}
  n2 = TextureSample(UVs=@n1) {texture=&T_ShieldCaustics}
  n3 = Constant()
  n4 = VertexColor()
  n5 = Multiply(A=@n4.R, B=@n4.R)
  n6 = Multiply(
   A=Add(
    A=TextureSample(UVs=@n1) {texture=&T_Inky_Smoke_Tile}.R
    B=ShadingPathSwitch(Default=@n2.R, Mobile=@n3) {desc=Migrated from FeatureLevelSwitch}
   )
   B=@n5
  )
  n7 = Multiply(A=OneMinus(None=DepthFade() {fadeDistanceDefault=5}), B=DepthFade() {fadeDistanceDefault=2})
  n8 = Multiply(A=@n6, B=ShadingPathSwitch(Default=Multiply(A=@n7) {constB=4}, Mobile=Constant() {r=1}) {desc=Migrated from FeatureLevelSwitch})
  n9 = ParticleColor()
  n10 = Multiply(
   A=LinearInterpolate(
    A=Constant3Vector() {constant=<rgba 0.5104 0.1601 0.021 0>}
    B=Constant3Vector() {constant=<rgba 1 0.5789 0.0524 0>}
    Alpha=Add(A=@n6, B=@n8)
   )
   B=@n9.RGB
  )
  n11 = Multiply(A=Time()) {constB=5}
  n12 = Multiply(A=Sine(None=@n11) {period=0.155}, B=Sine(None=@n11) {period=0.2})
  n13 = Sine(None=@n11) {period=0.43}
  n14 = LinearInterpolate(Alpha=Multiply(A=Sine(None=@n11) {period=0.075}, B=Multiply(A=@n12, B=@n13))) {constA=4}
  n15 = Constant() {r=1}
  MaterialFunctionCall(Emissive=MaterialFunctionCall(
    LDR Input=Saturate(None=@n10)
    HDR Dynamic Range=ShadingPathSwitch(Default=@n14, Mobile=@n15) {desc=Migrated from FeatureLevelSwitch}
    HDR Tint=@n10
   ) {materialFunction=&DeriveHDRfromLDR}.RemappedRange) {materialFunction=&MF_FX_EmissiveAdjustments}
  n16 = Fresnel()
  Multiply(
   A=Multiply(A=@n9.A, B=@n5)
   B=ShadingPathSwitch(Default=OneMinus(None=@n16), Mobile=Constant() {r=1}) {desc=Migrated from FeatureLevelSwitch}
  )
 ]
}
