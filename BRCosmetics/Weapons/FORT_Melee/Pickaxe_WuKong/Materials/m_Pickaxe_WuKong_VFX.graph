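# Additive unlit material for glowing mesh-particle VFX on the WuKong pickaxe.
# Emissive color combines a thresholded diffuse green-channel mask, normal-mapped Fresnel and panning smoke, flickered by multiplied sine waves and tinted by particle color.
# Switchable Fresnel and UV1 sphere-mask terms combine with particle alpha for opacity, while a dynamic parameter pushes vertices along normals.
# Mobile shading paths replace the flicker and both Fresnel terms with constants.
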Material /BRCosmetics/Weapons/FORT_Melee/Pickaxe_WuKong/Materials/m_Pickaxe_WuKong_VFX
graph {
 properties {
  bEnableResponsiveAA=true
  bUsedWithMeshParticles=true
  bUsedWithNiagaraMeshParticles=true
  blendMode=BLEND_Additive
  pixelDepthOffsetMode=PDOM_Legacy
  shadingModel=MSM_Unlit
  usedShadingModels=Unlit
 }
 nodes [
  n1 = Multiply(A=Time()) {constB=5}
  n2 = Sine(None=@n1) {period=0.075}
  n3 = Multiply(A=Multiply(A=Sine(None=@n1) {period=0.155}, B=Sine(None=@n1) {period=0.2}), B=Sine(None=@n1) {period=0.43})
  n4 = ShadingPathSwitch(Default=LinearInterpolate(Alpha=Multiply(A=@n2, B=@n3)) {constA=1.5, constB=0.9}, Mobile=Constant() {r=1}) {desc=Migrated from FeatureLevelSwitch}
  n5 = TextureSample(UVs=Panner(Coordinate=TextureCoordinate() {coordinateIndex=1, vTiling=2}) {speedY=0.5}) {texture=&T_Inky_Smoke_Tile}
  n6 = VectorParameter() {parameterName=EmissiveColor, defaultValue=<rgba 3 1.453 0.4394 0>}
  n7 = ScalarParameter() {defaultValue=0.2, parameterName=EmissiveAmount}
  n8 = Constant3Vector() {constant=<rgba 0 0 1 0>}
  n9 = TextureSample() {samplerType=SAMPLERTYPE_Normal, texture=/BRCosmetics/Weapons/FORT_Melee/Pickaxe_WuKong/Textures/T_Pickaxe_WuKong_N}
  n10 = DynamicParameter() {defaultValue=<rgba 0.15 1 1 1>, paramNames [WPO,Param2,Param3,Param4]}
  n11 = Fresnel(Normal=Transform(None=LinearInterpolate(A=@n8, B=@n9.RGB, Alpha=@n10.Param2)) {transformType=TRANSFORM_Local}) {exponent=1}
  n12 = Constant() {r=1}
  n13 = ScalarParameter() {parameterName=TextureBias}
  n14 = LinearInterpolate(Alpha=Power(Base=TextureSample!(UVs, Tex, Bias=@n13, Apply View MipBias) {mipValueMode=TMVM_MipBias, texture=/BRCosmetics/Weapons/FORT_Melee/Pickaxe_WuKong/Textures/T_Pickaxe_WuKong_D}.G) {constExponent=3}) {constB=4}
  n15 = Constant()
  n16 = Multiply(
   A=Multiply(A=Multiply(A=@n6.RGB, B=@n7), B=ShadingPathSwitch(Default=@n11, Mobile=@n12) {desc=Migrated from FeatureLevelSwitch})
   B=If(A=@n14, B=Constant() {r=0.05}, "A > B"=@n14, "A == B"=@n15, "A < B"=@n15)
  )
  n17 = Add(A=Multiply(A=Multiply(A=@n5.R) {constB=6}, B=@n16), B=@n16)
  MaterialFunctionCall(Emissive=Multiply(A=ParticleColor().RGB, B=Multiply(A=@n4, B=@n17))) {materialFunction=&MF_FX_EmissiveAdjustments}
  n18 = Fresnel() {exponent=1}
  n19 = Multiply(A=OneMinus(None=@n18)) {constB=3}
  n20 = Multiply(A=@n18) {constB=3}
  n21 = Power(Base=ComponentMask(None=TextureCoordinate() {coordinateIndex=1}) {g=true}) {constExponent=1.7}
  n22 = Constant() {r=0.5}
  n23 = Power(
   Base=OneMinus(None=SphereMask(A=@n21, B=@n22) {attenuationRadius=0.5, hardnessPercent=0})
   Exp=ScalarParameter() {defaultValue=5, parameterName=SphereMaskExp}
  ) {constExponent=5}
  n24 = Constant() {r=1}
  Multiply(
   A=ShadingPathSwitch(Default=StaticSwitchParameter(True=@n19, False=@n20) {parameterName=InvertFresnel}, Mobile=Constant() {r=1}) {desc=Migrated from FeatureLevelSwitch}
   B=Multiply(A=StaticSwitchParameter(True=@n23, False=@n24) {defaultValue=true, parameterName=UesMask}, B=ParticleColor().A)
  )
  Multiply(A=VertexNormalWS(), B=DynamicParameter() {defaultValue=<rgba 0.15 1 1 1>, paramNames [WPO,Param2,Param3,Param4]}.WPO) {constB=0.15}
 ]
}
