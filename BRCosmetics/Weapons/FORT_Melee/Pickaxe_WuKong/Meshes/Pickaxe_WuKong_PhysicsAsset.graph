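# Physics asset for the Pickaxe_WuKong two-handed melee pickaxe cosmetic, defining its simplified collision shape.
# It holds a single body on the Melee_TwoHanded_Handle bone, built from one capsule offset and rotated to run along the weapon.
# With only that body, it defines no constraints.
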PhysicsAsset /BRCosmetics/Weapons/FORT_Melee/Pickaxe_WuKong/Meshes/Pickaxe_WuKong_PhysicsAsset
bodies {Melee_TwoHanded_Handle [{shapeType=Capsule, length=79.8944, radius=7.7774, center=<0.7627 -0.9787 34.4973>, rotation=<rot 0 -90 0>}]}
constraints []
